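# Translucent UI material for a pip in the Iron Log weapon reticle, drawn as a two-color circle.
# A scalable MF_UI_Circle fill supplies the shape, with a PipOpacity parameter for fading.
# MainColor and AccentColor blend through a mask built from a second, oversized, offset circle.
# A static sine of horizontal UV displaces vertical UV to create wavy coordinates for that color split, without textures.
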Material /WildEstateWeapons/UI/IronLogReticle/Materials/M_UI_IronLog_Pip
graph {
 properties {bAutomaticallySetUsageInEditor=false, blendMode=BLEND_Translucent, materialDomain=MD_UI, usedShadingModels=Default Lit}
 nodes [
  LinearInterpolate(
   A=VectorParameter() {parameterName=MainColor, defaultValue=<rgba 1 0.0742 0.1946 1>}.RGB
   B=VectorParameter() {parameterName=AccentColor, defaultValue=<rgba 0.9823 0.7084 0 1>}.RGB
   Alpha=NamedRerouteUsage()
  )
  Multiply(A=NamedRerouteUsage(), B=ScalarParameter() {defaultValue=1, parameterName=PipOpacity})
  NamedRerouteDeclaration(None=MaterialFunctionCall(Size=ScalarParameter() {defaultValue=0.8, parameterName=Pip scale}) {materialFunction=&MF_UI_Circle}.Fill) {name=Circle, nodeColor=<rgba 0 0 0 1>}
  n1 = NamedRerouteUsage()
  n2 = Add(A=ComponentMask(None=@n1) {r=true}, B=Constant() {r=0.4}) {constB=0}
  n3 = Add(A=ComponentMask(None=@n1) {g=true}, B=Constant() {r=0.9}) {constB=0}
  NamedRerouteDeclaration(None=MaterialFunctionCall(UVs=AppendVector(A=@n2, B=@n3), Size=Constant() {r=2}) {materialFunction=&MF_UI_Circle}.Fill) {name=WaveMask, nodeColor=<rgba 0 0 0 1>}
  n4 = TextureCoordinate()
  n5 = Add(
   A=Multiply(A=ComponentMask(None=@n4) {r=true}, B=Constant() {desc=frequency X, r=3}) {constB=0.5}
   B=Multiply(A=ComponentMask(None=@n4) {g=true}, B=Constant() {desc=frequency Y}) {constB=0.5}
  ) {constB=0.5}
  n6 = Multiply(A=Sine(None=@n5), B=Constant()) {constB=0}
  n7 = Multiply(A=Sine(None=@n5), B=Constant() {r=0.08}) {constB=0}
  NamedRerouteDeclaration(None=Add(A=TextureCoordinate(), B=AppendVector(A=@n6, B=@n7))) {name=UV_Wave, nodeColor=<rgba 0 0 0 1>}
 ]
}
