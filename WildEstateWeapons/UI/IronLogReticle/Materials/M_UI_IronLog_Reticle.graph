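# Translucent UI material that draws the IronLog weapon reticle: four brackets, a circle and a center dot, with outlines.
# The Spread parameter feeds shared reticle UV functions with a rotation to offset the brackets.
# A bracket distance-field texture is thresholded twice for antialiased fill and outline, while circle functions draw the ring and dot procedurally.
# Named reroutes carry these masks into black-and-white color and opacity blends, with opacity parameters for brackets, outline and circle.
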Material /WildEstateWeapons/UI/IronLogReticle/Materials/M_UI_IronLog_Reticle
graph {
 properties {bAutomaticallySetUsageInEditor=false, blendMode=BLEND_Translucent, materialDomain=MD_UI, usedShadingModels=Default Lit}
 nodes [
  NamedRerouteDeclaration(None=MaterialFunctionCall(Size=Constant() {r=0.15}, OutlineThickness=Constant() {r=0.015}) {materialFunction=&MF_UI_Circle}.Outline) {name=Circle, nodeColor=<rgba 0 0 0 1>}
  n1 = Constant() {r=0.15}
  n2 = Constant() {r=0.03}
  NamedRerouteDeclaration(None=Power(Base=MaterialFunctionCall(Size=@n1, OutlineThickness=@n2) {materialFunction=&MF_UI_Circle}.Outline) {constExponent=0.01}) {name=Circle_Outline, nodeColor=<rgba 0 0 0 1>}
  n3 = TextureSample(UVs=NamedRerouteUsage()) {samplerSource=SSM_Clamp_WorldGroupSettings, texture=/WildEstateWeapons/UI/IronLogReticle/Textures/T_UI_IronLog_Reticle_Bracket}
  NamedRerouteDeclaration(None=MaterialFunctionCall(DistanceField=@n3.RGB, Threshold=Constant() {r=0.4}) {materialFunction=&MF_UI_AntiAliasing}.Outer) {name=Reticle, nodeColor=<rgba 0 0 0 1>}
  n4 = MaterialFunctionCall(Size=Constant() {r=0.025}, OutlineThickness=Constant() {r=0.02}) {materialFunction=&MF_UI_Circle}
  NamedRerouteDeclaration(None=@n4.Fill) {name=Center Dot, nodeColor=<rgba 0.7578 1 0 1>}
  NamedRerouteDeclaration(None=MaterialFunctionCall(DistanceField=@n3.RGB, Threshold=Constant() {r=0.25}) {materialFunction=&MF_UI_AntiAliasing}.Outer) {name=Outline, nodeColor=<rgba 0 0 0 1>}
  NamedRerouteDeclaration(None=Power(Base=@n4.Outline) {constExponent=0.01}) {name=Center dot outline, nodeColor=<rgba 0.9219 0 1 1>}
  n5 = NamedRerouteUsage()
  n6 = Constant3Vector() {constant=<rgba 1 1 1 0>}
  n7 = Multiply(A=NamedRerouteUsage(), B=@n6)
  n8 = LinearInterpolate(A=Constant3Vector(), B=@n6, Alpha=NamedRerouteUsage())
  NamedRerouteDeclaration(None=Add(A=Multiply(A=@n5, B=@n6), B=Add(A=@n7, B=@n8))) {name=Color, nodeColor=<rgba 1 0 0.3047 1>}
  n9 = Constant() {r=4}
  n10 = Constant() {r=0.5}
  n11 = MaterialFunctionCall!(
   Positional Spread Offset=ScalarParameter() {defaultValue=60, parameterName=Spread}
   Accuracy Spread Offset
   Min Value
   "Input Range [Min,Max]"
   "Remapped Range [Min,Max]"
  ) {materialFunction=/FortUILibrary/Systems/Reticles/MaterialFunctions/MF_UI_Reticle_Remap_Spread}
  NamedRerouteDeclaration(None=MaterialFunctionCall(
    UVs=MaterialFunctionCall!(
     Reticle UVs
     Reticle Sides=@n9
     Reticle Scale
     Reticle Rotation
     Prong Scale=@n10
     Prong Rotation
     Spread Offset=@n11.Result
     Positional Offset
     "Spread Input Range [Min,Max]"
     "Spread Remapped Range [Min,Max]"
    ) {materialFunction=/FortUILibrary/Systems/Reticles/MaterialFunctions/MF_UI_Reticle_UVs}.UVs
    "Rotation Angle (0-1)"=Constant() {r=0.125}
   ) {materialFunction=&CustomRotator}.Rotated Values) {name=CustomUV, nodeColor=<rgba 0 0 0 1>}
  n12 = ScalarParameter() {defaultValue=0.5, parameterName=Outline Opacity}
  n13 = Add(
   A=Multiply(A=ScalarParameter() {defaultValue=0.5, parameterName=Brackets opacity}, B=NamedRerouteUsage())
   B=Multiply(A=NamedRerouteUsage(), B=@n12)
  )
  n14 = NamedRerouteUsage()
  n15 = LinearInterpolate(A=Multiply(A=@n12, B=NamedRerouteUsage()), B=@n14, Alpha=@n14)
  n16 = NamedRerouteUsage()
  n17 = LinearInterpolate(A=NamedRerouteUsage(), B=@n16, Alpha=@n16)
  n18 = ScalarParameter() {defaultValue=0.2, parameterName=Circle Opacity}
  NamedRerouteDeclaration(None=Add(A=Add(A=@n13, B=@n15), B=Multiply(A=@n17, B=@n18))) {name=Opacity, nodeColor=<rgba 0 0.625 1 1>}
  NamedRerouteUsage()
  NamedRerouteUsage()
 ]
}
