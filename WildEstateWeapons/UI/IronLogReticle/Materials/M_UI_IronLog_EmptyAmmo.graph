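# Translucent UI material for the IronLog weapon reticle's empty-ammo indicator, apparently the dim container shape.
# A box function draws a rounded, half-height shape.
# An inverted circle mask, built on rescaled UVs through a 2x1 symmetry function, supplies the cutouts.
# A Container opacity parameter feeds a multiply chain with rerouted inputs, apparently these two masks, over a constant color.
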Material /WildEstateWeapons/UI/IronLogReticle/Materials/M_UI_IronLog_EmptyAmmo
graph {
 properties {bAutomaticallySetUsageInEditor=false, blendMode=BLEND_Translucent, materialDomain=MD_UI, usedShadingModels=Default Lit}
 nodes [
  Constant3Vector()
  Multiply(
   A=Multiply(A=ScalarParameter() {defaultValue=0.5, parameterName=Container opacity}, B=NamedRerouteUsage()) {constA=0.7}
   B=NamedRerouteUsage()
  )
  NamedRerouteDeclaration(None=MaterialFunctionCall(Size=Constant2Vector() {g=0.5, r=1}, CornerRoundness=Constant() {r=0.5}) {materialFunction=&MF_UI_Box}.Fill) {name=Shape, nodeColor=<rgba 0 0 0 1>}
  n1 = TextureCoordinate()
  n2 = ComponentMask(None=@n1) {r=true}
  n3 = Constant() {r=1}
  n4 = Multiply(A=ComponentMask(None=@n1) {g=true}, B=Constant() {r=2})
  n5 = Constant() {r=-0.5}
  n6 = MaterialFunctionCall!(UV=AppendVector(A=Multiply(A=@n2, B=@n3), B=Add(A=@n4, B=@n5)), Mirror Y, Swizzle UV) {materialFunction=/FortUILibrary/MaterialFunctions/Basics/MF_UI_Symmetry_2x1_New}
  n7 = Constant() {r=0.7}
  NamedRerouteDeclaration(None=OneMinus(None=MaterialFunctionCall(UVs=@n6.Symmetry, Size=@n7) {materialFunction=&MF_UI_Circle}.Fill)) {name=CircleMask, nodeColor=<rgba 0 0 0 1>}
 ]
}
