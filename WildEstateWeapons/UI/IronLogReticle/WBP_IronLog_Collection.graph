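# Charge stack pip collection widget for the IronLog weapon reticle, derived from the shared RechargeStackBase PipCollection reticle widget.
# An empty-ammo material fills the background, behind a fixed-size box holding a DynamicEntryBox of WBP_IronLog_Pip entries with radial spacing settings.
# On Construct it only passes that entry box to the parent through SetDynamicEntryBox, leaving pip logic to the base class.
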WidgetBlueprint /WildEstateWeapons/UI/IronLogReticle/WBP_IronLog_Collection
graphs [EventGraph {nodes [
   Event Construct()
   > SetDynamicEntryBox!(
    execute
    self:Fort Weapon Reticle Recharge Stack Base Pip Collection Base Object Reference
    DynamicEntryBox:Dynamic Entry Box Object Reference=Get DynamicEntryBox_Stacks() -> (DynamicEntryBox_Stacks:Dynamic Entry Box Object Reference)
   ) -> (then)
  ]}]
variables []
widgets {
 info {
  inheritedWidgetCount=0
  parentClass=/Game/Athena/HUD/reticles/ReticleExtensionWidgets/ChargeStack/WBP_Reticle_RechargeStackBase_PipCollection.WBP_Reticle_RechargeStackBase_PipCollection_C
  rootWidgetClass=&Overlay
  widgetCount=4
 }
 widgets [
  Overlay_0 {widgetClassPath=&Overlay, widget=Overlay_0}
  Image_Background {
   widgetClassPath=&Image
   bIsVariable=true
   parent=Overlay_0
   widget=Image_Background
   properties {brush {
     bIsDynamicallyLoaded=false
     drawAs=Image
     imageType=FullColor
     mirroring=NoMirror
     resourceName=None
     resourceObject=/WildEstateWeapons/UI/IronLogReticle/Materials/M_UI_IronLog_EmptyAmmo
     tiling=NoTile
     imageSize=<32 32>
     margin=<ltrb 0 0 0 0>
     outlineSettings {
      bUseBrushTransparency=false
      roundingType=HalfHeightRadius
      width=0
      color {colorUseRule=UseColor_Specified, specifiedColor=<rgba 0 0 0 0>}
      cornerRadii=<0 0 0 0>
     }
     tintColor {colorUseRule=UseColor_Specified, specifiedColor=<rgba 1 1 1 1>}
     uVRegion {bIsValid=false, max=<0 0>, min=<0 0>}
    }}
   slot {class=OverlaySlot, properties {horizontalAlignment=HAlign_Fill, verticalAlignment=VAlign_Fill}}
  }
  SizeBox_0 {
   widgetClassPath=&SizeBox
   parent=Overlay_0
   widget=SizeBox_0
   properties {bOverride_HeightOverride=true, bOverride_WidthOverride=true, heightOverride=50, widthOverride=50}
   slot {class=OverlaySlot}
  }
  DynamicEntryBox_Stacks {
   widgetClassPath=&DynamicEntryBox
   bIsVariable=true
   parent=SizeBox_0
   widget=DynamicEntryBox_Stacks
   properties {
    entryWidgetClass=&/WBP_IronLog_Pip.WBP_IronLog_Pip_C
    numDesignerPreviewEntries=2
    radialBoxSettings {angleBetweenItems=18, bClockwiseOrder=false, bDistributeItemsEvenly=false, marginSize=0, sectorCentralAngle=360, startingAngle=301}
   }
   slot {class=SizeBoxSlot, properties {horizontalAlignment=HAlign_Center, verticalAlignment=VAlign_Center}}
  }
 ]
}
