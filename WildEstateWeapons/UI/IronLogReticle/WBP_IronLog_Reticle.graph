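# Reticle widget for the Iron Log shotgun, extending the recharge-stack reticle base with a spread-reactive crosshair and recharge pips.
# BindToWeaponEvents caches the weapon, and Tick feeds its scaled spread-to-FOV ratio into the reticle material's Spread parameter.
# Construct and ammo changes toggle a polling timer that runs while the magazine is below capacity.
# Cooldown updates pass the weapon item's remaining time to the pip collection through PipOpacity.
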WidgetBlueprint /WildEstateWeapons/UI/IronLogReticle/WBP_IronLog_Reticle
animations [Anim_Spread {
  binding_count=1
  length_seconds=5
  object=&self:Anim_Spread
  sequence {bindings [Image_Reticle {id=#1, possessed_class=Image, tracks []}]}
 }]
graphs [
 BindToWeaponEvents {nodes [
   n1 = BindToWeaponEvents!() -> (then, DoBind:bool, EquippedWeapon:Fort Weapon Object Reference) {is=title K2Node_FunctionEntry}
   > Parent: BindToWeaponEvents!(execute, DoBind:bool=@n1, EquippedWeapon:Fort Weapon Object Reference=@n1.EquippedWeapon) -> (then) {is=title K2Node_CallParentFunction}
   > n2 = Cast To B_Shotgun_WildEstate_IronLog!(execute, Object=@n1.EquippedWeapon) -> (then, CastFailed, AsB Shotgun Wild Estate Iron Log:B Shotgun Wild Estate Iron Log Object Reference) {is=title K2Node_DynamicCast}
   > Set WeaponRef(WeaponRef:B Shotgun Wild Estate Iron Log Object Reference=@n2) -> (Output_Get:B Shotgun Wild Estate Iron Log Object Reference)
  ]}
 EventGraph {nodes [
   n1 = Event OnLocalAmmoChanged!() -> (OutputDelegate, then, LocalCount:int=0, LocalRemaining:int=0) {is=title K2Node_Event}
   > Parent: OnLocalAmmoChanged!(execute, LocalCount:int=@n1.LocalCount, LocalRemaining:int=@n1.LocalRemaining) -> (then) {is=title K2Node_CallParentFunction}
   > UpdateCooldownQueryTimer(self:Fort Weapon Reticle Extension Widget Base Object Reference)
   Event Destruct()
   > Parent: Destruct()
   > HandleCooldownTimer!(execute, self:Fort Weapon Reticle Extension Widget Base Object Reference, bActivateTimer:bool=false, TimerInterval:float=0) -> (then)
   Event Tick()
   n2 = Get WeaponRef() -> (WeaponRef:B Shotgun Wild Estate Iron Log Object Reference)
   > Is Valid(InputObject=@n2)
   > n3 = GetDynamicMaterial(self=Get Image_Reticle() -> (Image_Reticle:Image Object Reference))
   n4 = Get OwningPlayerController() -> (OwningPlayerController:Fort Player Controller Object Reference)
   > SetScalarParameterValue(
    self=@n3
    ParameterName=Spread
    Value=float * float(
     A=GetSpreadToFieldOfViewRatio!(self:Fort Weapon Ranged Object Reference=@n2, PlayerController:Fort Player Controller Object Reference=@n4, bUseMinimumSpread:bool=false) -> (ReturnValue:float)
     B=1000
    ) {is=title K2Node_PromotableOperator}
   )
   Event Construct()
   > Parent: Construct()
   > n5 = Cast To FortPlayerController(Object=GetOwningPlayer())
   > Set OwningPlayerController(OwningPlayerController:Fort Player Controller Object Reference=@n5) -> (Output_Get:Fort Player Controller Object Reference)
   > UpdateCooldownQueryTimer()
   Event UpdateRemainingCooldown() {is=title K2Node_Event}
   > n6 = GetRemainingCooldownTimer!(execute, self) -> (then, RemainingTime:double, TotalDuration:double)
   > UpdateRemainingRechargeTime!(
    execute
    self:Fort Weapon Reticle Recharge Stack Base Pip Collection Base Object Reference=GetPipCollection!(self:WBP Reticle Recharge Stack Base Object Reference) -> (PipCollection:WBP Reticle Recharge Stack Base Pip Collection Object Reference)
    TimeRemaining:float=@n6
    TotalDuration:float=@n6.TotalDuration
    PipPercentMatParamName:Name=PipOpacity
    ShouldSetPipColor:bool=false
    NotReadyColor:Linear Color Structure
    ReadyColor:Linear Color Structure
    PipColorMatParamName:Name=None
   ) -> (then)
   Event UpdateCooldownQueryTimer() {is=title K2Node_Event}
   > n7 = Get!(execute:Exec) -> (then:Exec, else:Exec, WeaponRef:B Shotgun Wild Estate Iron Log Object Reference)
   > n8 = Branch(Condition="integer >= integer"(
     A=GetLocalMagazineAmmoCount!(self:Fort Weapon Object Reference=@n7) -> (ReturnValue:int=0)
     B=GetBulletsPerClip!(self:Fort Weapon Object Reference=@n7) -> (ReturnValue:int=0)
    ))
   > HandleCooldownTimer(bActivateTimer=false, TimerInterval=0)
   HandleCooldownTimer(execute=@n8.else, bActivateTimer=true, TimerInterval=0.1)
  ]}
 GetPipCollection {nodes [
   GetPipCollection() {is=title K2Node_FunctionEntry}
   > Return Node!(
    execute
    PipCollection:WBP Reticle Recharge Stack Base Pip Collection Object Reference=Get WBP_IronLog_Collection() -> (WBP_IronLog_Collection:WBP Iron Log Collection Object Reference)
   )
  ]}
 GetRemainingCooldownTimer {nodes [
   GetRemainingCooldownTimer() {is=title K2Node_FunctionEntry}
   > n1 = Cast To FortPlayerController(Object=GetOwningPlayer())
   > n2 = Get()
   n3 = GetCoolDownTimeRemaining!(
    self:Fort Item Object Reference=GetInventoryItem!(self:Fort Weapon Object Reference=@n2) -> (ReturnValue:Fort Item Object Reference)
    FortPC:Fort Player Controller Object Reference=@n1
   ) -> (CooldownDuration:float, ReturnValue:float)
   > Return Node!(execute, RemainingTime:double=@n3.ReturnValue, TotalDuration:double=@n3)
   Return Node(execute=[@n2.else, @n1.CastFailed], RemainingTime=0, TotalDuration=0)
  ]}
]
variables [
 OwningPlayerController {type=FortPlayerController, default=null}
 WeaponRef {type=B_Shotgun_WildEstate_IronLog_C, default=null}
]
widgets {
 info {
  inheritedWidgetCount=1
  parentClass=/Game/Athena/HUD/reticles/ReticleExtensionWidgets/ChargeStack/WBP_Reticle_RechargeStackBase.WBP_Reticle_RechargeStackBase_C
  rootWidgetClass=&Overlay
  widgetCount=4
 }
 widgets [
  Overlay_0 {widgetClassPath=&Overlay, widget=Overlay_0}
  Image_Reticle {
   widgetClassPath=&Image
   bIsVariable=true
   parent=Overlay_0
   widget=Image_Reticle
   properties {brush {
     bIsDynamicallyLoaded=false
     drawAs=Image
     imageType=FullColor
     mirroring=NoMirror
     resourceName=None
     resourceObject=/WildEstateWeapons/UI/IronLogReticle/Materials/M_UI_IronLog_Reticle
     tiling=NoTile
     imageSize=<128 128>
     margin=<ltrb 0 0 0 0>
     outlineSettings {
      bUseBrushTransparency=false
      roundingType=HalfHeightRadius
      width=0
      color {colorUseRule=UseColor_Specified, specifiedColor=<rgba 0 0 0 0>}
      cornerRadii=<0 0 0 0>
     }
     tintColor {colorUseRule=UseColor_Specified, specifiedColor=<rgba 1 1 1 1>}
     uVRegion {bIsValid=false, max=<0 0>, min=<0 0>}
    }}
   slot {class=OverlaySlot}
  }
  WBP_IronLog_Collection {
   widgetClassPath=&/WBP_IronLog_Collection
   bIsVariable=true
   parent=Overlay_0
   widget=WBP_IronLog_Collection
   properties {renderTransform {angle=0, scale=<0.7 0.7>, shear=<0 0>, translation=<0 32>}}
   slot {class=OverlaySlot, properties {horizontalAlignment=HAlign_Center, verticalAlignment=VAlign_Bottom}}
  }
  TutorialHighlight {widgetClassPath=&TutorialHighlight}
 ]
}
